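# Niagara trail effect for the Dark Minion glider, combining smoke ribbons and drifting feather meshes.
# Two anim-trail ribbons follow bone or socket pairs, their gameplay opacity scaled by a system FN_Integrator value.
# That value also gates gameplay feathers, which track skeletal sockets with rotated offsets while slightly lerping back toward spawn position.
# Front-end modifiers add ribbon velocity and enable a separate feather emitter using two offset boxes instead of sockets.
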NiagaraSystem /BRCosmetics/Gadgets/Assets/VinderTech_GliderChute/Glider_Dark_Minion_Male/VFX/NS_Glider_DarkMinion_Trails_FeathersAndSmoke
dependencies {
 usedDataInterfaces [&NiagaraDataInterfaceCurve, &NiagaraDataInterfaceColorCurve]
 usedDynamicInputs []
 usedModules [&SystemState, &FN_Integrator]
}
emitters {
 FN_AnimTrail_Smoke {
  emitterUpdateScript [
   &EmitterState(Scalability Mode=::NewEnumerator1)
   /Game/Effects/Niagara/Modules/AnimTrail/FN_SpawnAnimTrail/FN_SpawnAnimTrail(Number of Anim Trails=2, SpawnRate=100)
  ]
  particleSpawnScript [
   /Game/Effects/Niagara/Modules/AnimTrail/FN_InitializeAnimTrail/FN_InitializeAnimTrailBoneIndecies(Second Bone or Socket Index=2)
   &FN_InitializeAnimTrailBoneIndecies(AnimTrail Selection=1, First Bone or Socket Index=1, Second Bone or Socket Index=3) {module=FN_InitializeAnimTrailBoneIndecies001}
   /Game/Effects/Niagara/Modules/AnimTrail/FN_InitializeAnimTrail/FN_InitializeAnimTrail(
    AnimTrailWidth=1
    Color=&MultiplyLinearColor(Linear Color=<rgba 1 1 1 1.5>, Scale Factor=<1 1 1 1>)
    Lifetime=0.12
    Write Material Random=false
   )
   &CameraOffset(Camera Offset Amount=&RandomRangeFloat(Minimum=5, Maximum=15))
   &AddVelocity(Velocity=&FN_FrontEndVector3Modifier(Gameplay=<0 0 0>, Front End=<0 1000 0>))
  ]
  particleUpdateScript [
   &InheritVelocity(Module.Limit Speed=true, Velocity Limit=20)
   &SolveForcesAndVelocity(Clamp Velocity=false, Limit Acceleration=false)
   /Game/Effects/Niagara/Modules/AnimTrail/FN_UpdateAnimTrail/FN_UpdateAnimTrail(New float=1)
   &ScaleColor(
    Scale Alpha=&Multiply_Float(
     A=&FloatFromCurve(FloatCurve=DataInterface() {propertyValues {Curve {keys [
          {arriveTangent=3.6646, interpMode=RCIM_Cubic, leaveTangent=3.6646, tangentMode=RCTM_User, time=-0.0017, value=0.5086}
          {arriveTangent=-0.0143, interpMode=RCIM_Cubic, leaveTangent=-0.0143, tangentMode=RCTM_User, time=0.1233, value=0.7411}
          {arriveTangent=-1.317, interpMode=RCIM_Cubic, leaveTangent=-1.317, tangentMode=RCTM_User, time=1, value=0}
         ]}}})
     B=&FN_FrontEndFloatModifier(
      Gameplay=&FloatFromCurve(
       FloatCurve=DataInterface() {propertyValues {Curve {keys [<at 0 0 RCIM_Cubic>, {arriveTangent=2.0248, interpMode=RCIM_Cubic, leaveTangent=2.0248, tangentMode=RCTM_User, time=1, value=1}]}}}
       CurveIndex=$System.FN_Integrator.Value
      )
      Front End=1
     )
    )
    Scale RGB=&MakeVectorFromLinearColorRGB(Color=<rgba 1.4879 10 0 1>)
    ScaleA=true
    ScaleRGB=false
   )
  ]
  rendererClasses [&NiagaraRibbonRendererProperties]
  renderers [{rendererClass=&NiagaraRibbonRendererProperties}]
 }
 Feathers {
  emitterUpdateScript [
   &EmitterState()
   &SpawnRate(
    Spawn Probability=0.6
    SpawnRate=&MakeCustomFloatFromBool(Bool=&SetBoolByFloatComparison(A=$System.FN_Integrator.Value, B=0.6), True Float=2, False Float=0)
    Use Spawn Probability=true
   )
   &SpawnBurst_Instantaneous(
    Spawn Count=4
    Spawn Probability=&MakeCustomFloatFromBool(Bool=&SetBoolByFloatComparison(A=$System.FN_Integrator.Value, B=0.6), True Float=0.4, False Float=0)
    Use Spawn Probability=true
   )
  ]
  particleSpawnScript [
   &InitializeParticle(
    Color=<rgba 0.9167 0.9167 0.9167 1>
    Lifetime=&RandomRangeFloat(Minimum=0.8, Maximum=1.25)
    Mesh Scale Mode=::NewEnumerator1
    Mesh Uniform Scale Max=0.5
    Mesh Uniform Scale Min=0.24
    UsePositionOffset=false
    Write Material Random=true
   )
   &FN_SetSpawnGroup(Number of Spawn Groups=2)
   /SetVariables(Particles.ParticlePositionOffset:Vector3f=&MaskVectorBySpawnGroup(
     Masked Vector=&RandomRangeVector(Minimum=<-3 60 0>, Maximum=<3 25 6>)
     Passthrough Vector=&RandomRangeVector(Minimum=<-3 -25 0>, Maximum=<3 -60 6>)
     Spawn Group Mask=1
    )) {isSetParameters=true}
   &SkeletalMeshLocation(
    Apply To Particle Position=false
    Bone / Socket Index=&SpawnGroup()
    Inherit Velocity=false
    MaskSpawnGroup=false
    Mesh Sampling Type=::NewEnumerator5
    Skeletal Mesh=$User.SkeletalMesh
    Socket Sampling Mode=::NewEnumerator5
   )
   &CameraOffset(Camera Offset Amount=&RandomRangeFloat(Minimum=-5, Maximum=10))
   &InitialMeshOrientation(
    Rotation=&RandomRangeVector(Minimum=<0 0 0>, Maximum=<1 1 1>)
    Use Orientation Vector=true
    Use Rotation Vector=true
   )
  ]
  particleUpdateScript [
   &ParticleState()
   &DynamicMaterialParameters(
    Index 0 Param 1=&RandomRangeFloat(Minimum=0, Maximum=0.9)
    Index 0 Param 2=&FloatFromCurve(FloatCurve=DataInterface() {propertyValues {Curve {keys [<at 0.6 1>,<at 1 0>]}}})
    Module.Param0WriteEnabled=true
    Module.Param1WriteEnabled=true
    Module.Param2WriteEnabled=true
    Module.Param3WriteEnabled=true
   ) {module=DynamicMaterialParameters001}
   &SkeletalMeshLocation(
    Apply To Particle Position=false
    Bone / Socket Index=&SpawnGroup()
    Inherit Velocity=false
    Interpolate Mesh Sampling=false
    MaskSpawnGroup=false
    Mesh Sampling Type=::NewEnumerator5
    Skeletal Mesh=$User.SkeletalMesh
    Socket Sampling Mode=::NewEnumerator5
   ) {module=SkeletalMeshLocation001}
   /SetVariables(Particles.ParticlePositionOffsetRotate:Vector3f=&RotateVectorByQuaternion(Vector To Rotate=$Particles.Initial.ParticlePositionOffset, Quaternion=$Output.SkeletalMeshLocation001.SampledRotation)) {isSetParameters=true}
   &self:LERP-Source_to_Destination(
    DestinationPosition:NiagaraPosition=$Particles.Initial.Position
    DestinationPosition-Offset:Vector3f=$Particles.ParticlePositionOffsetRotate
    Normalized Influence:NiagaraFloat=&FloatFromCurve(
     FloatCurve=DataInterface() {propertyValues {Curve {keys [
         {arriveTangent=0.0353, interpMode=RCIM_Cubic, leaveTangent=0.0353, tangentMode=RCTM_User, time=0, value=0}
         {arriveTangent=0.0217, interpMode=RCIM_Cubic, leaveTangent=0.0217, tangentMode=RCTM_User, time=1, value=0.15}
        ]}}}
     Scale Curve=&RandomRangeFloat(Minimum=0.5, Maximum=1)
    )
    SourcePosition:Vector3f=$Output.SkeletalMeshLocation001.SampledPosition
    SourcePosition-Offset:Vector3f=$Particles.ParticlePositionOffsetRotate
   ) {module=LERP-Source_to_Destination}
   &ScaleColor(
    Scale Mode=::NewEnumerator1
    Scale RGBA=&Vector4FromCurve(Vector4Curve=DataInterface() {propertyValues {
       AlphaCurve {keys [<at 0.0133 0>,<at 0.2264 1>,<at 0.7428 1>,<at 1 0>]}
       BlueCurve {keys [<at 0.0165 1>,<at 0.7713 0.6>]}
       GreenCurve {keys [<at 0.0165 1>,<at 0.7713 0.6>]}
       RedCurve {keys [<at 0.0165 1>,<at 0.7713 0.6>]}
      }})
    ScaleRGBA=true
   )
   &UpdateMeshOrientation(Rotation Vector=&RandomRangeVector(Minimum=<-1 -1 -1>, Maximum=<1 1 1>))
  ]
  rendererClasses [&NiagaraMeshRendererProperties]
  renderers [{rendererClass=&NiagaraMeshRendererProperties, rendererIndex=0}]
 }
 Feathers_FE {
  emitterUpdateScript [
   &EmitterState()
   &SpawnRate(
    Spawn Probability=0.6
    SpawnRate=&FN_FrontEndFloatModifier(Gameplay=0, Front End=2)
    Use Spawn Probability=true
   )
   &SpawnBurst_Instantaneous(
    Spawn Count=&MakeIntFromFloat(Float=&FN_FrontEndFloatModifier(Gameplay=0, Front End=4))
    Spawn Probability=0.5
    Use Spawn Probability=true
   )
  ]
  particleSpawnScript [
   &InitializeParticle(
    Color=<rgba 0.9167 0.9167 0.9167 1>
    Lifetime=&RandomRangeFloat(Minimum=0.8, Maximum=1.25)
    Mesh Scale Mode=::NewEnumerator1
    Mesh Uniform Scale Max=0.5
    Mesh Uniform Scale Min=0.24
    UsePositionOffset=false
    Write Material Random=true
   )
   &FN_SetSpawnGroup(Number of Spawn Groups=2)
   &CameraOffset(Camera Offset Amount=&RandomRangeFloat(Minimum=-5, Maximum=10))
   &InitialMeshOrientation(
    Rotation=&RandomRangeVector(Minimum=<0 0 0>, Maximum=<1 1 1>)
    Use Orientation Vector=true
    Use Rotation Vector=true
   )
   &BoxLocation(
    Box Offset=&MaskVectorBySpawnGroup(Masked Vector=<-150 0 0>, Passthrough Vector=<150 0 0>)
    Box Size=<60 20 20>
    MaskSpawnGroup=false
    Module.Override Seed=false
    Rotate Box=false
    Surface Only=false
   )
   &AddVelocity(Velocity=<0 180 0>)
  ]
  particleUpdateScript [
   &Drag(Use Linear Drag=true, Use Rotational Drag=false)
   &SolveForcesAndVelocity(Clamp Velocity=false, Limit Acceleration=false)
   &ParticleState()
   &DynamicMaterialParameters(
    Index 0 Param 1=&RandomRangeFloat(Minimum=0, Maximum=0.9)
    Index 0 Param 2=&FloatFromCurve(FloatCurve=DataInterface() {propertyValues {Curve {keys [<at 0.6 1>,<at 1 0>]}}})
    Module.Param0WriteEnabled=true
    Module.Param1WriteEnabled=true
    Module.Param2WriteEnabled=true
    Module.Param3WriteEnabled=true
   ) {module=DynamicMaterialParameters001}
   &ScaleColor(
    Scale Mode=::NewEnumerator1
    Scale RGBA=&Vector4FromCurve(Vector4Curve=DataInterface() {propertyValues {
       AlphaCurve {keys [<at 0.0133 0>,<at 0.2264 1>,<at 0.7428 1>,<at 1 0>]}
       BlueCurve {keys [<at 0.0165 1>,<at 0.7713 0.6>]}
       GreenCurve {keys [<at 0.0165 1>,<at 0.7713 0.6>]}
       RedCurve {keys [<at 0.0165 1>,<at 0.7713 0.6>]}
      }})
    ScaleRGBA=true
   )
   &UpdateMeshOrientation(Rotation Vector=&RandomRangeVector(Minimum=<-1 -1 -1>, Maximum=<1 1 1>))
  ]
  rendererClasses [&NiagaraMeshRendererProperties]
  renderers [{rendererClass=&NiagaraMeshRendererProperties, rendererIndex=0}]
 }
}
summary {
 systemName=NS_Glider_DarkMinion_Trails_FeathersAndSmoke
 emitters [
  {emitterName=Feathers, rendererClasses [&NiagaraMeshRendererProperties]}
  {emitterName=FN_AnimTrail_Smoke, rendererClasses [&NiagaraRibbonRendererProperties]}
  {emitterName=Feathers_FE, rendererClasses [&NiagaraMeshRendererProperties]}
 ]
 userVariables [
  User.bIsFrontEnd {type=NiagaraBool}
  User.bIsFrontEnd {type=NiagaraFloat, defaultValue=0, description=""}
  User.bIsFrontEndPreview {type=NiagaraBool}
  User.ForwardDot {type=NiagaraFloat, defaultValue=0, description=""}
  User.IsFullyDeployed {type=NiagaraBool}
  User.RibbonIDOffset {type=NiagaraInt32}
  User.RightDot {type=NiagaraFloat, defaultValue=0, description=""}
  User.SkeletalMesh {
   type=NiagaraDataInterfaceSkeletalMesh
   defaultValue {
    struct=NiagaraExt_VariableValue_DataInterface
    value {dataInterface=&self:NiagaraDataInterfaceSkeletalMesh_0}
   }
  }
 ]
}
